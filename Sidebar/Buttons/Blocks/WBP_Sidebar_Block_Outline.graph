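# Sidebar button block, derived from the UIKit block base, that draws a material-driven focus outline.
# Transition events animate it with hover, unhover and sheen animations plus a hover sound, or set material parameters directly for instant changes.
# Input method changes adjust sheen alpha, dimming it for gamepad.
# Functions set corner radii, reset animation state and toggle whether the outline extends beyond the block through negative padding.
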WidgetBlueprint /Sidebar/Buttons/Blocks/WBP_Sidebar_Block_Outline
animations [
 Anim_Hover {
  binding_count=1
  length_seconds=1
  object=&self:Anim_Hover
  sequence {bindings [FocusOutline {
     id=#1
     possessed_class=Image
     tracks [
      {
       type=MovieSceneFloatTrack
       display_name=Render Opacity
       sections [{type=MovieSceneFloatSection, channels [None {type=MovieSceneScriptingFloatChannel, keys [[2,0],[3,1]]}]}]
      }
      {
       type=MovieSceneWidgetMaterialTrack
       sections [{
         type=MovieSceneParameterSection
         channels [
          Anim_Hover_Fill {type=MovieSceneScriptingFloatChannel, keys [[0,0],[2,0],[20,1]]}
          Anim_Hover_Energy {type=MovieSceneScriptingFloatChannel, keys [[2,0],[17,1]]}
          Anim_Hover_Outline {type=MovieSceneScriptingFloatChannel, keys [[2,0],[7,1]]}
          Anim_OutlineThickness {type=MovieSceneScriptingFloatChannel, keys [[2,0],[7,1]]}
         ]
        }]
      }
     ]
    }]}
 }
 Anim_Unhover {
  binding_count=1
  length_seconds=0.3417
  object=&self:Anim_Unhover
  sequence {
   bindings [FocusOutline {
     id=#2
     possessed_class=Image
     tracks [
      {
       type=MovieSceneFloatTrack
       display_name=Render Opacity
       sections [{type=MovieSceneFloatSection, channels [None {type=MovieSceneScriptingFloatChannel, keys [[0,1],[12,1],[41,0]]}]}]
      }
      {
       type=MovieSceneWidgetMaterialTrack
       sections [{
         type=MovieSceneParameterSection
         channels [
          Anim_Hover_Fill {type=MovieSceneScriptingFloatChannel, const [1,1]}
          Anim_Hover_Energy {type=MovieSceneScriptingFloatChannel, const [1,1]}
          Anim_Hover_Outline {type=MovieSceneScriptingFloatChannel, keys [[0,0.5],[24,1]]}
          Anim_OutlineThickness {type=MovieSceneScriptingFloatChannel, keys [[0,1],[24,0]]}
         ]
        }]
      }
     ]
    }]
   display_rate {numerator=120}
  }
 }
 Anim_Sheen {
  binding_count=1
  length_seconds=1.5
  object=&self:Anim_Sheen
  sequence {bindings [FocusOutline {
     id=#3
     possessed_class=Image
     tracks [{
       type=MovieSceneWidgetMaterialTrack
       sections [{
         type=MovieSceneParameterSection
         channels [
          Anim_Highlight {type=MovieSceneScriptingFloatChannel, keys [[0,0.5],[30,1]]}
          Sheen_Falloff {type=MovieSceneScriptingFloatChannel, keys [[0,0],[9,1],[30,0]]}
         ]
        }]
      }]
    }]}
 }
]
graphs [
 ContainOutline {nodes [
   n1 = ContainOutline!() -> (then, ContainOutlinePadding:bool=~) {is=title K2Node_FunctionEntry}
   > n2 = Set Contain Outline Padding(Contain Outline Padding:bool=@n1) -> (Output_Get:bool)
   > Set Padding!(
    execute
    Padding:Margin Structure=Select!(Option 0_Left:float=-16, Option 0_Top:float=-16, Option 0_Right:float=-16, Option 0_Bottom:float=-16, Option 1:Margin Structure, Index=@n2) -> (ReturnValue:Margin Structure)
    self:Grid Slot Object Reference=SlotAsGridSlot(Widget=Get FocusOutline() -> (FocusOutline:Image Object Reference))
   ) -> (then, Output_Get:Margin Structure)
  ]}
 EventGraph {nodes [
   n1 = Focused_Visualization!() -> (OutputDelegate, then, NeedsToAnimate:bool=~) {is=title K2Node_CustomEvent}
   > n2 = Branch(Condition=@n1.NeedsToAnimate)
   > n3 = Branch(Condition=IsAnimationPlaying(InAnimation=Get Anim_Unhover() -> (Anim_Unhover:Widget Animation Object Reference)))
   > n4 = StopAnimation(InAnimation=Get Anim_Unhover())
   SetRenderOpacity(execute=[@n4.then, @n3.else], self=Self-Reference(), InOpacity=1)
   > PlayAnimationForward(InAnimation=Get Anim_Hover() -> (Anim_Hover:Widget Animation Object Reference))
   > PlayAnimationForward(InAnimation=Get Anim_Sheen() -> (Anim_Sheen:Widget Animation Object Reference))
   > SpawnSound2D!(
    execute
    Sound:Sound Base Object Reference=Get BlockOutlineHoverSound() -> (BlockOutlineHoverSound:Sound Base Object Reference)
    VolumeMultiplier:float=1
    PitchMultiplier:float=1
    StartTime:float=0
    ConcurrencySettings:Sound Concurrency Object Reference
    bPersistAcrossLevelTransition:bool=false
    bAutoDestroy:bool=true
   ) -> (then, ReturnValue:Audio Component Object Reference)
   n5 = GetDynamicMaterial(execute=@n2.else, self=Get FocusOutline())
   > SetScalarParameterValue(self=@n5, ParameterName=Get Anim_Hover_Fill() -> (Anim_Hover_Fill:Name=None), Value=1)
   > SetScalarParameterValue(self=@n5, ParameterName=Get Anim_Hover_Outline() -> (Anim_Hover_Outline:Name=None), Value=1)
   > SetScalarParameterValue(self=@n5, ParameterName=Get Anim_Hover_Energy() -> (Anim_Hover_Energy:Name=None), Value=1)
   > SetScalarParameterValue(self=@n5, ParameterName=Get Anim_OutlineThickness() -> (Anim_OutlineThickness:Name=None), Value=1)
   > SetScalarParameterValue(self=@n5, ParameterName=Get Anim_Highlight() -> (Anim_Highlight:Name=None), Value=1)
   > SetRenderOpacity(self=Get FocusOutline(), InOpacity=1)
   n6 = Unfocused_Visualization!() -> (OutputDelegate, then, NeedsToAnimate:bool=~) {is=title K2Node_CustomEvent}
   > n7 = Branch(Condition=@n6.NeedsToAnimate)
   > n8 = Branch(Condition=IsAnimationPlaying(InAnimation=Get Anim_Hover()))
   > n9 = StopAnimation(InAnimation=Get Anim_Hover())
   SetRenderOpacity(execute=[@n9.then, @n8.else], self=Self-Reference(), InOpacity=1)
   > PlayAnimationForward(InAnimation=Get Anim_Unhover())
   n10 = GetDynamicMaterial(execute=@n7.else, self=Get FocusOutline())
   > SetScalarParameterValue(self=@n10, ParameterName=Get Anim_Hover_Fill() -> (Anim_Hover_Fill=None), Value=1)
   > SetScalarParameterValue(self=@n10, ParameterName=Get Anim_Hover_Outline() -> (Anim_Hover_Outline=None), Value=1)
   > SetScalarParameterValue(self=@n10, ParameterName=Get Anim_Hover_Energy() -> (Anim_Hover_Energy=None), Value=1)
   > SetScalarParameterValue(self=@n10, ParameterName=Get Anim_OutlineThickness() -> (Anim_OutlineThickness=None), Value=0)
   > SetRenderOpacity(self=Get FocusOutline(), InOpacity=0)
   Event Construct()
   > SetRenderOpacity(self=[Self-Reference(), Get FocusOutline()], InOpacity=0)
   n11 = Create Event()
   n12 = Get CommonInputSubsystem()
   > Bind Event to On Input Method Changed(self=@n12, Delegate=@n11)
   > OnHandleInputMethodChanged!(execute, self, Input Method:ECommonInputType Enum=GetCurrentInputType(self=@n12)) -> (then)
   Event PreConstruct()
   > ContainOutline!(execute, self, ContainOutlinePadding:bool=Get Contain Outline Padding() -> (Contain Outline Padding:bool)) -> (then) {is=title K2Node_CallFunction}
   Event OnInstantTransitionToFocused() {is=title K2Node_Event}
   > StopAllAnimations(self:User Widget Object Reference)
   > SetIsFocused!(execute, self:WBP UIKit Block Base Object Reference, IsFocused:bool=true) -> (then)
   > Focused_Visualization!(execute, self, NeedsToAnimate:bool=false) -> (then) {is=title K2Node_CallFunction}
   Event OnTransitionFocusedRequested() {is=title K2Node_Event}
   > SetIsFocused(IsFocused=true)
   > Branch(Condition=IsVisible())
   > Focused_Visualization(NeedsToAnimate=true)
   Event OnTransitionUnfocusedRequested() {is=title K2Node_Event}
   > Branch(Condition=GetIsFocused!() -> (IsFocused:bool))
   > SetIsFocused(IsFocused=false)
   > Branch(Condition=IsVisible())
   > Unfocused_Visualization!(execute, self, NeedsToAnimate:bool=true) -> (then) {is=title K2Node_CallFunction}
   n13 = Event OnTransitionDisabledRequested() {is=title K2Node_Event}
   Stop All Transitions!(In:Exec=[@n13.then, @n14.then, @n15.then]) -> (Out:Exec) {is=title K2Node_MacroInstance}
   > SetRenderOpacity(self=[Self-Reference(), Get FocusOutline()], InOpacity=0)
   n14 = Event OnInstantTransitionToEnabled() {is=title K2Node_Event}
   n15 = Event OnInstantTransitionToDisabled() {is=title K2Node_Event}
   n16 = Event OnVisibilitySet!() -> (OutputDelegate, then, IsVisible:bool) {is=title K2Node_Event}
   > n17 = Branch(Condition=@n16.IsVisible)
   > Branch(Condition=GetIsFocused())
   > RequestTransition!(execute, self:UIKit Block Object Reference, State:EUIKitBlockVisualState Enum=Focused) -> (then)
   RequestInstantTransition!(execute=@n17.else, self:UIKit Block Object Reference, State:EUIKitBlockInstantTransitionState Enum=InstantDisabled) -> (then)
   Event Destruct()
   > Unbind Event from On Input Method Changed(self=@n12, Delegate=@n11)
   Event OnInstantTransitionToUnfocused() {is=title K2Node_Event}
   > StopAllAnimations()
   > FlushAnimations(self:User Widget Object Reference)
   > SetIsFocused(IsFocused=false)
   > Unfocused_Visualization(NeedsToAnimate=false)
  ]}
 OnHandleInputMethodChanged {nodes [
   OnHandleInputMethodChanged!() -> (then, Input Method:ECommonInputType Enum) {is=title K2Node_FunctionEntry}
   n1 = Get FocusOutline()
   > SetScalarParameterValue(
    self=GetDynamicMaterial!(Brush:"Slate Brush Structure (by ref)"=Get Brush!(self:Image Object Reference=@n1) -> (Brush:Slate Brush Structure)) -> (ReturnValue)
    ParameterName=Sheen_Alpha
    Value=Select!(
     MouseAndKeyboard:float=0.5
     Gamepad:float=0.1
     Touch:float=0.5
     Count:float=0.5
     Index:ECommonInputType Enum=Get Input Method() -> (Input Method:ECommonInputType Enum)
    ) -> (ReturnValue:float)
   )
  ]}
 ResetAnimationState {nodes [
   ResetAnimationState() {is=title K2Node_FunctionEntry}
   > n1 = Branch(Condition=IsAnyAnimationPlaying!(self:User Widget Object Reference) -> (ReturnValue:bool))
   > StopAllAnimations()
   > n2 = FlushAnimations()
   Unfocused_Visualization(execute=[@n2.then, @n1.else], NeedsToAnimate=false)
   > SetRenderOpacity(self=Self-Reference(), InOpacity=0)
  ]}
 SetCornerRadiuses {nodes [
   n1 = SetCornerRadiuses!() -> (then, TopLeft:float=~, TopRight:float=~, BottomLeft:float=~, BottomRight:float=~, RadiusInPixels:bool=~) {is=title K2Node_FunctionEntry}
   > Set TopLeftCornerRadius(TopLeftCornerRadius:double=@n1) -> (Output_Get:double)
   > Set TopRightCornerRadius(TopRightCornerRadius:double=@n1.TopRight) -> (Output_Get:double)
   > Set BottomLeftCornerRadius(BottomLeftCornerRadius:double=@n1.BottomLeft) -> (Output_Get:double)
   > Set BottomRightCornerRadius(BottomRightCornerRadius:double=@n1.BottomRight) -> (Output_Get:double)
   > Set RadiusInPixels(RadiusInPixels:bool=@n1.RadiusInPixels) -> (Output_Get:bool)
   > n2 = GetDynamicMaterial!(execute:Exec, self:Image Object Reference=Get FocusOutline()) -> (then:Exec, ReturnValue)
   > SetScalarParameterValue(self=@n2, ParameterName=Radius in Pixels, Value="To Float (Boolean)"(InBool=Get RadiusInPixels() -> (RadiusInPixels:bool)))
   > SetVectorParameterValue(
    self=@n2
    ParameterName=Corner Radius
    Value=Make Linear Color!(
     R:float=Get TopLeftCornerRadius() -> (TopLeftCornerRadius:double)
     G:float=Get BottomLeftCornerRadius() -> (BottomLeftCornerRadius:double)
     B:float=Get TopRightCornerRadius() -> (TopRightCornerRadius:double)
     A:float=Get BottomRightCornerRadius() -> (BottomRightCornerRadius:double)
    ) -> (LinearColor:Linear Color Structure) {is=title K2Node_MakeStruct}
   )
  ]}
]
variables [
 Anim_Highlight {type=string, default=Anim_Highlight}
 Anim_Hover_Energy {type=string, default=Anim_Hover_Energy}
 Anim_Hover_Fill {type=string, default=Anim_Hover_Fill}
 Anim_Hover_Outline {type=string, default=Anim_Hover_Outline}
 Anim_OutlineThickness {type=string, default=Anim_OutlineThickness}
 BlockOutlineHoverSound {type=SoundBase, default=null}
 BottomLeftCornerRadius {type=number, default=0.3}
 BottomRightCornerRadius {type=number, default=0.3}
 Contain Outline Padding {type=boolean}
 RadiusInPixels {type=boolean}
 TopLeftCornerRadius {type=number, default=0.3}
 TopRightCornerRadius {type=number, default=0.3}
]
widgets {
 info {
  inheritedWidgetCount=0
  parentClass=/FortUILibrary/Systems/ModularButton/Blocks/WBP_UIKit_Block_Base.WBP_UIKit_Block_Base_C
  rootWidgetClass=&GridPanel
  widgetCount=2
 }
 widgets [
  Root {widgetClassPath=&GridPanel, widget=Root, properties {columnFill [1], rowFill [1]}}
  FocusOutline {
   widgetClassPath=&Image
   bIsVariable=true
   parent=Root
   widget=FocusOutline
   properties {
    pixelSnapping=Disabled
    visibility=SelfHitTestInvisible
    brush {
     bIsDynamicallyLoaded=false
     drawAs=Image
     imageType=FullColor
     mirroring=NoMirror
     resourceName=None
     resourceObject=/Sidebar/Materials/MI_Sidebar_Outline
     tiling=NoTile
     imageSize=<32 32>
     margin=<ltrb 0 0 0 0>
     outlineSettings {
      bUseBrushTransparency=false
      roundingType=HalfHeightRadius
      width=0
      color {colorUseRule=UseColor_Specified, specifiedColor=<rgba 0 0 0 0>}
      cornerRadii=<0 0 0 0>
     }
     tintColor {colorUseRule=UseColor_Specified, specifiedColor=<rgba 1 1 1 1>}
     uVRegion {bIsValid=false, max=<0 0>, min=<0 0>}
    }
   }
   slot {class=GridSlot, properties {padding=<ltrb -16 -16 -16 -16>}}
  }
 ]
}
