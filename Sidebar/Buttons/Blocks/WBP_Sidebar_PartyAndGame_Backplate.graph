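# Sidebar backplate widget for party and game blocks, layering a material-driven Backplate image and a collapsed Divider in a grid.
# Functions set dynamic material parameters for color, corner radius and fade, with PreConstruct applying ShowFade.
# The color enum maps to a scalar switch, with one entry remapped for touch input.
# Other functions toggle backplate padding and show the widget only for party, native Epic party, team or game voice header types.
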WidgetBlueprint /Sidebar/Buttons/Blocks/WBP_Sidebar_PartyAndGame_Backplate
graphs [
 ContainBackplate {nodes [
   n1 = ContainBackplate!() -> (then, ContainBackplatePadding:bool=~) {is=title K2Node_FunctionEntry}
   > Set Padding!(
    execute
    Padding:Margin Structure=Select!(
     Option 0_Left:float=0
     Option 0_Top:float=0
     Option 0_Right:float=0
     Option 0_Bottom:float=0
     Option 1_Left:float=16
     Option 1_Top:float=16
     Option 1_Right:float=16
     Option 1_Bottom:float=16
     Index=@n1
    ) -> (ReturnValue:Margin Structure)
    self:Grid Slot Object Reference=SlotAsGridSlot(Widget=Get Backplate() -> (Backplate:Image Object Reference))
   ) -> (then, Output_Get:Margin Structure)
  ]}
 EventGraph {nodes [
   Event PreConstruct()
   > SetShowFade!(execute, self, Show Fade:bool=Get ShowFade() -> (ShowFade:bool)) -> (then)
  ]}
 Set Party And Game Backplate Visibility {nodes [
   n1 = Set Party And Game Backplate Visibility!() -> (then, Header Item:ESocialHeaderItemType Enum=Unknown) {is=title K2Node_FunctionEntry}
   n2 = "Equal (Enum)"(A:ESocialHeaderItemType Enum=@n1, B:ESocialHeaderItemType Enum=PartyMembers)
   n3 = "Equal (Enum)"(A=@n1, B=GameVoiceChannels)
   n4 = "Equal (Enum)"(A=@n1, B=NativeEpicPartyMembers)
   n5 = "Equal (Enum)"(A=@n1, B=TeamMembers)
   > SetVisibility(InVisibility=Select!(
     Option 0:ESlateVisibility Enum=Collapsed
     Option 1:ESlateVisibility Enum=SelfHitTestInvisible
     Index=OR Boolean!(A=@n5, B=@n4, C:bool=@n2, D:bool=@n3) -> (ReturnValue)
    ) -> (ReturnValue:ESlateVisibility Enum))
  ]}
 SetBackplateColor {nodes [
   n1 = SetBackplateColor!() -> (then, Backplate Switch:E_UI_Sidebar_Button_Backplate_Color Enum=NewEnumerator0) {is=title K2Node_FunctionEntry}
   n2 = Get Backplate()
   n3 = IsUsingInputType!(InputType:ECommonInputType Enum=Touch, PlayerController:Player Controller Object Reference=GetOwningPlayer()) -> (ReturnValue:bool)
   > SetScalarParameterValue(
    self=GetDynamicMaterial!(Brush:"Slate Brush Structure (by ref)"=Get Brush!(self:Image Object Reference=@n2) -> (Brush:Slate Brush Structure)) -> (ReturnValue)
    ParameterName=Backplate Switch
    Value=Select!(
     NewEnumerator0:float=0
     NewEnumerator1:float=1
     NewEnumerator2:float=2
     NewEnumerator3:float=3
     NewEnumerator4:float=SelectFloat(A=1, B=4, bPickA=@n3)
     NewEnumerator5:float=5
     NewEnumerator6:float=6
     NewEnumerator7:float=7
     NewEnumerator8:float=8
     NewEnumerator9:float=9
     NewEnumerator11:float=10
     NewEnumerator12:float=11
     NewEnumerator13:float=12
     Index:E_UI_Sidebar_Button_Backplate_Color Enum=@n1
    ) -> (ReturnValue:float)
   )
  ]}
 SetCornerRadius {nodes [
   SetCornerRadius!() -> (then, Radius In Pixels:bool=~, Corner Radius:double=~) {is=title K2Node_FunctionEntry}
   n1 = Get Backplate()
   > SetScalarParameterValue(
    self=GetDynamicMaterial(Brush=Get Brush(self=@n1))
    ParameterName=Radius In Pixels
    Value="To Float (Boolean)"(InBool=Get Radius In Pixels() -> (Radius In Pixels:bool))
   )
   n2 = Get Backplate()
   > SetScalarParameterValue(self=GetDynamicMaterial(Brush=Get Brush(self=@n2)), ParameterName=Corner Radius, Value=Get Corner Radius() -> (Corner Radius:double))
  ]}
 SetShowFade {nodes [
   n1 = SetShowFade!() -> (then, Show Fade:bool=~) {is=title K2Node_FunctionEntry}
   n2 = Get Backplate()
   > SetScalarParameterValue(self=GetDynamicMaterial(Brush=Get Brush(self=@n2)), ParameterName=Show Fade, Value="To Float (Boolean)"(InBool=@n1))
  ]}
]
variables [ShowFade {type=boolean, default=true}]
widgets {
 info {inheritedWidgetCount=0, parentClass=&UserWidget, rootWidgetClass=&GridPanel, widgetCount=3}
 widgets [
  Root {widgetClassPath=&GridPanel, widget=Root, properties {columnFill [1], rowFill [1]}}
  Backplate {
   widgetClassPath=&Image
   bIsVariable=true
   parent=Root
   widget=Backplate
   properties {
    visibility=SelfHitTestInvisible
    brush {
     bIsDynamicallyLoaded=false
     drawAs=Image
     imageType=FullColor
     mirroring=NoMirror
     resourceName=None
     resourceObject=/Sidebar/Materials/M_Sidebar_Backplate_PartyAndGame
     tiling=NoTile
     imageSize=<32 32>
     margin=<ltrb 0 0 0 0>
     outlineSettings {
      bUseBrushTransparency=false
      roundingType=HalfHeightRadius
      width=0
      color {colorUseRule=UseColor_Specified, specifiedColor=<rgba 0 0 0 0>}
      cornerRadii=<0 0 0 0>
     }
     tintColor {colorUseRule=UseColor_Specified, specifiedColor=<rgba 1 1 1 1>}
     uVRegion {bIsValid=false, max=<0 0>, min=<0 0>}
    }
   }
   slot {class=GridSlot}
  }
  Divider {
   widgetClassPath=&Image
   bIsVariable=true
   parent=Root
   widget=Divider
   properties {
    visibility=Collapsed
    brush {
     bIsDynamicallyLoaded=false
     drawAs=Image
     imageType=FullColor
     mirroring=NoMirror
     resourceName=None
     resourceObject=/Sidebar/Materials/MI_Sidebar_Divider
     tiling=NoTile
     imageSize=<32 2>
     margin=<ltrb 0 0 0 0>
     outlineSettings {
      bUseBrushTransparency=false
      roundingType=HalfHeightRadius
      width=0
      color {colorUseRule=UseColor_Specified, specifiedColor=<rgba 0 0 0 0>}
      cornerRadii=<0 0 0 0>
     }
     tintColor {colorUseRule=UseColor_Specified, specifiedColor=<rgba 1 1 1 1>}
     uVRegion {bIsValid=false, max=<0 0>, min=<0 0>}
    }
   }
   slot {class=GridSlot, properties {rowSpan=1, verticalAlignment=VAlign_Bottom}}
  }
 ]
}
